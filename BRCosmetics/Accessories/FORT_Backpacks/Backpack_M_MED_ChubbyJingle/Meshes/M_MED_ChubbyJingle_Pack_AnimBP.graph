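# Animation Blueprint for the ChubbyJingle backpack, adding spring-driven secondary motion to its cookie and sock bones.
# Each update maps the spine_05-to-spine_01 yaw delta to Parashuting, the alpha of an additive offset and roll on the pack bone.
# That pose feeds chained root-relative AnimDynamics nodes, simulating the center cookie and sock bones twice with different joint offsets.
# It also sets Not_frontEnd, which no visible node reads.
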AnimBlueprint /BRCosmetics/Accessories/FORT_Backpacks/Backpack_M_MED_ChubbyJingle/Meshes/M_MED_ChubbyJingle_Pack_AnimBP
target_skeleton=&/M_MED_ChubbyJingle_Pack_Skeleton
graphs [
 AnimGraph {nodes [
   Anim Dynamics - Bone: Dyn_coockie_C(Alpha=1, ExternalForce="0,0,0") {
    angularDampingOverride=0.999
    angularSpringConstant=500
    bAngularSpring=true
    bLinearSpring=true
    bOverrideAngularDamping=true
    bOverrideLinearDamping=true
    bUseGravityOverride=true
    linearDampingOverride=0.999
    linearSpringConstant=100
    simulationSpace=RootRelative
    boundBone {boneName=Dyn_coockie_C}
    componentAppliedLinearAccClamp=<600 600 600>
    componentLinearAccScale=<0.6 0.6 0.6>
    componentLinearVelScale=<0.6 0.6 0.6>
    physicsBodyDefinitions [{
      boundBone {boneName=Dyn_coockie_C}
      boxExtents=<10 10 10>
      constraintSetup {angularLimitsMax=<0 5 3>, angularLimitsMin=<0 -5 -3>, angularTarget=<1 0 0>, linearAxesMax=<1.5 0 0>}
      localJointOffset=<1 0 0>
     }]
   }
   "Transform (Modify) Bone - Bone: pack"(Translation="0,0,0", Rotation="0,0,0") {
    rotationMode=BMM_Additive
    scaleMode=BMM_Additive
    translationMode=BMM_Additive
    alphaBoolBlend {blendOption=Linear}
    boneToModify {boneName=pack}
   }
   n1 = "Transform (Modify) Bone - Bone: pack"(
    ComponentPose=Local To Component(LocalPose=Input Pose())
    Alpha=Get Parashuting() -> (Parashuting:double)
    Translation="0,-5,0.5"
    Rotation="0.000,0.000,40.000"
   ) {
    rotationMode=BMM_Additive
    scaleMode=BMM_Additive
    translationMode=BMM_Additive
    alphaBoolBlend {blendOption=Linear}
    boneToModify {boneName=pack}
    rotation {roll=40}
    translation {y=-5, z=0.5}
   }
   n2 = Anim Dynamics - Bone: Dyn_coockie_Fr_01(ComponentPose=@n1, Alpha=1, ExternalForce="0,0,0") {
    angularDampingOverride=0.999
    angularSpringConstant=500
    bAngularSpring=true
    bLinearSpring=true
    bOverrideAngularDamping=true
    bOverrideLinearDamping=true
    bUseGravityOverride=true
    linearDampingOverride=0.999
    linearSpringConstant=50
    simulationSpace=RootRelative
    boundBone {boneName=Dyn_coockie_Fr_01}
    componentAppliedLinearAccClamp=<600 600 600>
    componentLinearAccScale=<0.4 0.4 0.4>
    componentLinearVelScale=<0.4 0.4 0.4>
    gravityOverride {z=-600}
    physicsBodyDefinitions [{
      boundBone {boneName=Dyn_coockie_Fr_01}
      boxExtents=<9 9 9>
      constraintSetup {angularLimitsMax=<3 5 3>, angularLimitsMin=<-3 -5 0>, angularTarget=<1 0 0>, linearAxesMax=<0.3 0 0.4>}
      localJointOffset=<0.8 0 0>
     }]
   }
   n3 = Anim Dynamics - Bone: Dyn_coockie_Fr_02(ComponentPose=@n2, Alpha=1, ExternalForce="0,0,0") {
    angularDampingOverride=0.999
    angularSpringConstant=500
    bAngularSpring=true
    bLinearSpring=true
    bOverrideAngularDamping=true
    bOverrideLinearDamping=true
    bUseGravityOverride=true
    linearDampingOverride=0.999
    linearSpringConstant=80
    simulationSpace=RootRelative
    boundBone {boneName=Dyn_coockie_Fr_02}
    componentAppliedLinearAccClamp=<600 600 600>
    componentLinearAccScale=<0.4 0.4 0.4>
    componentLinearVelScale=<0.4 0.4 0.4>
    gravityOverride {z=-600}
    physicsBodyDefinitions [{
      boundBone {boneName=Dyn_coockie_Fr_02}
      boxExtents=<9 9 9>
      constraintSetup {angularLimitsMax=<3 5 3>, angularLimitsMin=<-3 -5 0>, angularTarget=<1 0 0>, linearAxesMax=<0.5 0 0.3>, linearAxesMin=<0 0 -0.3>}
      localJointOffset=<0.8 0 0>
     }]
   }
   n4 = Anim Dynamics - Bone: Dyn_coockie_Fr_03(ComponentPose=@n3, Alpha=1, ExternalForce="0,0,0") {
    angularDampingOverride=0.999
    angularSpringConstant=500
    bAngularSpring=true
    bLinearSpring=true
    bOverrideAngularDamping=true
    bOverrideLinearDamping=true
    bUseGravityOverride=true
    linearDampingOverride=0.999
    linearSpringConstant=70
    simulationSpace=RootRelative
    boundBone {boneName=Dyn_coockie_Fr_03}
    componentAppliedLinearAccClamp=<600 600 600>
    componentLinearAccScale=<0.4 0.4 0.4>
    componentLinearVelScale=<0.4 0.4 0.4>
    gravityOverride {z=-600}
    physicsBodyDefinitions [{
      boundBone {boneName=Dyn_coockie_Fr_03}
      boxExtents=<9 9 9>
      constraintSetup {angularLimitsMax=<3 5 3>, angularLimitsMin=<-3 -5 0>, angularTarget=<1 0 0>, linearAxesMax=<0.4 0 0.3>, linearAxesMin=<0 0 -0.3>}
      localJointOffset=<0.8 0 0>
     }]
   }
   n5 = Anim Dynamics - Bone: Dyn_coockie_Bk_01(ComponentPose=@n4, Alpha=1, ExternalForce="0,0,0") {
    angularDampingOverride=0.999
    angularSpringConstant=500
    bAngularSpring=true
    bLinearSpring=true
    bOverrideAngularDamping=true
    bOverrideLinearDamping=true
    bUseGravityOverride=true
    linearDampingOverride=0.999
    linearSpringConstant=150
    simulationSpace=RootRelative
    boundBone {boneName=Dyn_coockie_Bk_01}
    componentAppliedLinearAccClamp=<600 600 600>
    componentLinearAccScale=<0.4 0.4 0.4>
    componentLinearVelScale=<0.4 0.4 0.4>
    gravityOverride {z=-600}
    physicsBodyDefinitions [{
      boundBone {boneName=Dyn_coockie_Bk_01}
      boxExtents=<10 10 10>
      constraintSetup {angularLimitsMax=<3 3 3>, angularLimitsMin=<-3 -3 0>, angularTarget=<1 0 0>, linearAxesMax=<0.4 0 0.3>, linearAxesMin=<0 0 -0.3>}
      localJointOffset=<0.8 0 0>
     }]
   }
   n6 = Anim Dynamics - Bone: Dyn_coockie_Bk_02(ComponentPose=@n5, Alpha=1, ExternalForce="0,0,0") {
    angularDampingOverride=0.999
    angularSpringConstant=500
    bAngularSpring=true
    bLinearSpring=true
    bOverrideAngularDamping=true
    bOverrideLinearDamping=true
    bUseGravityOverride=true
    linearDampingOverride=0.999
    linearSpringConstant=150
    simulationSpace=RootRelative
    boundBone {boneName=Dyn_coockie_Bk_02}
    componentAppliedLinearAccClamp=<600 600 600>
    componentLinearAccScale=<0.4 0.4 0.4>
    componentLinearVelScale=<0.4 0.4 0.4>
    gravityOverride {z=-600}
    physicsBodyDefinitions [{
      boundBone {boneName=Dyn_coockie_Bk_02}
      boxExtents=<10 10 10>
      constraintSetup {angularLimitsMax=<3 3 3>, angularLimitsMin=<-3 -3 0>, angularTarget=<1 0 0>, linearAxesMax=<0.4 0 0.3>, linearAxesMin=<0 0 -0.3>}
      localJointOffset=<0.8 0 0>
     }]
   }
   n7 = Anim Dynamics - Bone: Dyn_coockie_C!(ComponentPose=@n6, bAlphaBoolEnabled:bool=True, ExternalForce="0,0,0") -> (Pose) {
    alphaInputType=Bool
    angularDampingOverride=0.999
    angularSpringConstant=800
    bAngularSpring=true
    bLinearSpring=true
    bOverrideAngularDamping=true
    bOverrideLinearDamping=true
    bUseGravityOverride=true
    linearDampingOverride=0.999
    linearSpringConstant=100
    simulationSpace=RootRelative
    boundBone {boneName=Dyn_coockie_C}
    componentAppliedLinearAccClamp=<500 500 500>
    componentLinearAccScale=<0.35 0.35 0.35>
    componentLinearVelScale=<0.35 0.35 0.35>
    gravityOverride {z=-350}
    physicsBodyDefinitions [{
      boundBone {boneName=Dyn_coockie_C}
      boxExtents=<10 10 10>
      constraintSetup {angularLimitsMax=<0 5 3>, angularLimitsMin=<0 -5 -3>, angularTarget=<1 0 0>, linearAxesMax=<2 0 0>}
      localJointOffset=<0.8 0 0>
     }]
   }
   n8 = Anim Dynamics - Bone: Dyn_socs!(ComponentPose=@n7, bAlphaBoolEnabled:bool=True, ExternalForce="0,0,0") -> (Pose) {
    alphaInputType=Bool
    angularDampingOverride=0.9999
    angularSpringConstant=200
    bAngularSpring=true
    bOverrideAngularDamping=true
    bUseGravityOverride=true
    simulationSpace=RootRelative
    boundBone {boneName=Dyn_socs}
    componentAppliedLinearAccClamp=<800 800 800>
    componentLinearAccScale=<0.6 0.6 0.6>
    componentLinearVelScale=<0.6 0.6 0.6>
    gravityOverride {z=-980}
    physicsBodyDefinitions [{
      boundBone {boneName=Dyn_socs}
      boxExtents=<10 10 10>
      constraintSetup {angularLimitsMax=<3 7 7>, angularLimitsMin=<-3 -7 -7>, angularTarget=<1 0 0>}
      localJointOffset=<0.5 0 0>
     }]
   }
   n9 = Anim Dynamics - Bone: Dyn_coockie_C(ComponentPose=@n8, bAlphaBoolEnabled=True, ExternalForce="0,0,0") {
    alphaInputType=Bool
    angularDampingOverride=0.999
    angularSpringConstant=800
    bAngularSpring=true
    bLinearSpring=true
    bOverrideAngularDamping=true
    bOverrideLinearDamping=true
    bUseGravityOverride=true
    linearDampingOverride=0.999
    linearSpringConstant=100
    simulationSpace=RootRelative
    boundBone {boneName=Dyn_coockie_C}
    componentAppliedLinearAccClamp=<500 500 500>
    componentLinearAccScale=<0.35 0.35 0.35>
    componentLinearVelScale=<0.35 0.35 0.35>
    gravityOverride {z=-350}
    physicsBodyDefinitions [{
      boundBone {boneName=Dyn_coockie_C}
      boxExtents=<10 10 10>
      constraintSetup {angularLimitsMax=<0 5 3>, angularLimitsMin=<0 -5 -3>, angularTarget=<1 0 0>, linearAxesMax=<2 0 0>}
      localJointOffset=<0.2 0 0>
     }]
   }
   n10 = Anim Dynamics - Bone: Dyn_socs(ComponentPose=@n9, bAlphaBoolEnabled=True, ExternalForce="0,0,0") {
    alphaInputType=Bool
    angularDampingOverride=0.9999
    angularSpringConstant=200
    bAngularSpring=true
    bOverrideAngularDamping=true
    bUseGravityOverride=true
    simulationSpace=RootRelative
    boundBone {boneName=Dyn_socs}
    componentAppliedLinearAccClamp=<800 800 800>
    componentLinearAccScale=<0.6 0.6 0.6>
    componentLinearVelScale=<0.6 0.6 0.6>
    gravityOverride {z=-980}
    physicsBodyDefinitions [{
      boundBone {boneName=Dyn_socs}
      boxExtents=<10 10 10>
      constraintSetup {angularLimitsMax=<3 7 7>, angularLimitsMin=<-3 -7 -7>, angularTarget=<1 0 0>}
      localJointOffset=<0.4 0 0>
     }]
   }
   Output Pose(Result=Component To Local(ComponentPose=@n10)) {layerGroup=None}
  ]}
 EventGraph {nodes [
   Event BlueprintUpdateAnimation()
   n1 = GetDeltaTransformFromRefPose!(self=GetOwnerSkeletalMeshComponent(), BoneName=spine_05, BaseName=spine_01) -> (ReturnValue_Location:Vector, ReturnValue_Rotation:Rotator, ReturnValue_Scale:Vector)
   > Set Parashuting(Parashuting:double=MapRangeClamped(Value=BreakRotator(InRot=@n1.ReturnValue_Rotation).Yaw, InRangeA=-15, InRangeB=-50)) -> (Output_Get:double)
   > Set Not_frontEnd(Not_frontEnd:bool=NOT Boolean(A=Get bIsInFrontEnd())) -> (Output_Get:bool)
   TryGetPawnOwner()
  ]}
]
variables [
 Not_frontEnd {type=boolean}
 Parashuting {type=number}
]
